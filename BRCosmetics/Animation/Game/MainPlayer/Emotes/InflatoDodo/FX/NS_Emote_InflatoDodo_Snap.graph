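# Niagara effect for the "Snap" moment of the InflatoDodo emote: a short impact-like burst of flash, streaks, sparks, rocks and dust.
# Seven burst emitters layer a flare, mesh and sprite streaks, dust puffs, and GPU-simulated sparks, sprite debris and rock meshes split across three renderers.
# Dust color and alpha follow time of day, and spark alpha halves in front-end preview.
# All emitters distance-cull, with rock count also scaled by distance.
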NiagaraSystem /BRCosmetics/Animation/Game/MainPlayer/Emotes/InflatoDodo/FX/NS_Emote_InflatoDodo_Snap
dependencies {
 usedDataInterfaces [
  &NiagaraDataInterfaceVectorCurve
  &NiagaraDataInterfaceCurve
  &NiagaraDataInterfaceVector2DCurve
  &NiagaraDataInterfaceSpriteRendererInfo
 ]
 usedDynamicInputs [&MakeFloatFromVector2d, &FNTimeOfDayScalar]
 usedModules [
  &SystemState
  &self:SystemSpawnScript.SystemScriptSource.SystemScriptGraph.NiagaraNodeAssignment_1.SetVariables_#1
 ]
}
emitters {
 Dust {
  emitterUpdateScript [
   &EmitterState(
    Enable Distance Culling=true
    Enable Visibility Culling=true
    Life Cycle Mode=::NewEnumerator1
    Loop Behavior=::NewEnumerator1
    Loop Delay=0.03
    Loop Duration=1
    Module.bUseMaxDistance=true
    Module.bUseMinDistance=false
    Scalability Mode=::NewEnumerator1
    UseLoopDelay=true
   )
   &SpawnBurst_Instantaneous(Spawn Count=2, Use Spawn Probability=false) {module=SpawnBurst_Instantaneous001}
  ]
  particleSpawnScript [
   &InitializeParticle(
    Color=<rgba 0.5255 0.4875 0.4453 1>
    Lifetime=&UniformRangedFloat(Minimum=0.75, Maximum=1)
    Position Offset=<-3 0 6>
    Sprite Rotation Angle=&UniformRangedFloat(Minimum=0, Maximum=1)
    Sprite Rotation Mode=::NewEnumerator1
    Sprite Size Mode=::NewEnumerator3
    Uniform Sprite Size=&UniformRangedFloat(Minimum=20, Maximum=23)
    UsePositionOffset=true
    Write Material Random=true
   ) {module=InitializeParticle001}
   /SetVariables(Particles.InitSpriteRotRate=&UniformRangedFloat(Minimum=-1, Maximum=1)) {isSetParameters=true}
   &AccelerationForce(Acceleration=&UniformRangedVector(Minimum=<0 0 -25>, Maximum=<0 0 -75>))
   &SphereLocation(
    MaskSpawnGroup=false
    Override Seed=false
    Sphere Radius=1
    Surface Only=true
    UseNonUniformScale=false
    UseOffset=true
    UseSphereOrientation=false
   )
   &ApplyInitialForces()
   &CameraOffset(Camera Offset Amount=&RandomRangeFloat(Minimum=0, Maximum=15))
   &AddVelocity(Velocity=&Multiply_Vector(
     A=$Output.SphereLocation.SphereNormal
     B=&VectorFromFloat(Value=&RandomRangeFloat(Minimum=20, Maximum=70))
    ))
  ]
  particleUpdateScript [
   &ParticleState()
   &SpriteRotationRate(Rotation Rate=&Multiply_Float(
     A=$Particles.InitSpriteRotRate
     B=&FloatFromCurve(
      FloatCurve=DataInterface() {propertyValues {Curve {keys [
          {arriveTangent=-1.9214, interpMode=RCIM_Cubic, leaveTangent=-1.9214, tangentMode=RCTM_User, time=-0.0046, value=1.0113}
          {arriveTangent=-0.2496, interpMode=RCIM_Cubic, leaveTangent=-0.2496, tangentMode=RCTM_User, time=0.9896, value=0.16}
         ]}}}
      CurveIndex=$Emitter.Age
     )
    ))
   &ScaleColor(
    Scale Alpha=&FloatFromCurve(
     FloatCurve=DataInterface() {propertyValues {Curve {keys [
         {arriveTangent=9.0725, interpMode=RCIM_Cubic, leaveTangent=9.0725, tangentMode=RCTM_User, time=0, value=0}
         {arriveTangent=-2.5249, interpMode=RCIM_Cubic, leaveTangent=-2.5249, tangentMode=RCTM_User, time=0.1408, value=0.9847}
         {arriveTangent=-0.0405, interpMode=RCIM_Cubic, leaveTangent=-0.0405, tangentMode=RCTM_User, time=1, value=0}
        ]}}}
     Scale Curve=&RandomRangeFloat(Minimum=0.8, Maximum=1)
    )
    Scale RGB=&VectorFromFloat(Value=&FloatFromCurve(
      FloatCurve=DataInterface() {propertyValues {Curve {keys [
          {arriveTangent=9.0725, interpMode=RCIM_Cubic, leaveTangent=9.0725, tangentMode=RCTM_User, time=0, value=1}
          {arriveTangent=-1.8242, interpMode=RCIM_Cubic, leaveTangent=-1.8242, tangentMode=RCTM_User, time=0.1408, value=2}
          {arriveTangent=-0.0405, interpMode=RCIM_Cubic, leaveTangent=-0.0405, tangentMode=RCTM_User, time=1, value=1}
         ]}}}
      Scale Curve=&RandomRangeFloat(Minimum=0.6, Maximum=0.7)
     ))
    ScaleA=true
    ScaleRGB=true
   ) {module=ScaleColor001}
   &GravityForce(Gravity=&VectorFromCurve(
     VectorCurve=DataInterface() {propertyValues {
       XCurve {keys [<at 0 1>,<at 1 1>]}
       YCurve {keys [<at 0 1>,<at 1 1>]}
       ZCurve {keys [{interpMode=RCIM_Cubic, tangentMode=RCTM_User, time=0, value=0}, {interpMode=RCIM_Cubic, tangentMode=RCTM_User, time=1, value=1}]}
      }}
     Scale Curve=<0 0 -50>
    ))
   &Drag(
    Drag=&FloatFromCurve(
     FloatCurve=DataInterface() {propertyValues {Curve {keys [
         {interpMode=RCIM_Cubic, tangentMode=RCTM_User, time=0, value=0}
         {interpMode=RCIM_Cubic, tangentMode=RCTM_User, time=0.5, value=1}
         {tangentMode=RCTM_User, time=1, value=0}
        ]}}}
     Scale Curve=&RandomRangeFloat(Minimum=7, Maximum=15)
    )
    Use Linear Drag=true
    Use Rotational Drag=false
   )
   &SolveForcesAndVelocity(Clamp Velocity=false, Limit Acceleration=false)
   &ScaleSpriteSize(Scale Factor=&Vector2DFromFloat(Value=&FloatFromCurve(FloatCurve=DataInterface() {propertyValues {Curve {keys [
          {arriveTangent=14.6311, interpMode=RCIM_Cubic, leaveTangent=14.6311, tangentMode=RCTM_User, time=0, value=0}
          {arriveTangent=5.5431, interpMode=RCIM_Cubic, leaveTangent=5.5431, tangentMode=RCTM_User, time=0.2, value=2}
          {arriveTangent=1.0299, interpMode=RCIM_Cubic, leaveTangent=1.0299, tangentMode=RCTM_User, time=1, value=3.7}
         ]}}})))
   &DynamicMaterialParameters(
    Index 0 Param 1=&FloatFromCurve(FloatCurve=DataInterface() {propertyValues {Curve {keys [<at 0 0 RCIM_Cubic>,<at 1.0011 0.2>]}}})
    Index 0 Param 2=&RandomRangeFloat(Minimum=1, Maximum=3)
    Index 0 Param 3=&FloatFromCurve(FloatCurve=DataInterface() {propertyValues {Curve {keys [{interpMode=RCIM_Cubic, leaveTangent=4.6779, time=0.2851, value=0}]}}})
    Module.Param0WriteEnabled=true
    Module.Param1WriteEnabled=true
    Module.Param2WriteEnabled=false
    Module.Param3WriteEnabled=false
   )
   &ScaleColor(
    Scale Alpha=$System.TOD
    Scale RGB=&FN_Select_Vector_from_Time_Of_Day(
     Morning=&MakeVectorFromLinearColorRGB(Color=<rgba 1 1 1 1>)
     Day=&MakeVectorFromLinearColorRGB(Color=<rgba 1 1 1 1>)
     Evening=&MakeVectorFromLinearColorRGB(Color=<rgba 0.9896 1.1844 2 1>)
     Night=&MakeVectorFromLinearColorRGB(Color=<rgba 0.5729 0.8481 2 1>)
    )
    ScaleA=true
    ScaleRGB=true
   )
   /SetVariables(Particles.SandColor:LinearColor=&Multiply_LinearColorByFloat(
     Linear Color=$Particles.Color
     Float=&FloatFromCurve(FloatCurve=DataInterface() {propertyValues {Curve {keys [{arriveTangent=-2.0035, interpMode=RCIM_Cubic, leaveTangent=-2.0035, tangentMode=RCTM_User, time=0, value=1.5}, <at 1 0>]}}})
    )) {isSetParameters=true}
   /SetVariables(Particles.SubImageIndexPoof:NiagaraFloat=&Lerp_Float(A=20, B=63, Alpha=$Particles.NormalizedAge)) {isSetParameters=true}
  ]
  rendererClasses [&NiagaraSpriteRendererProperties]
  renderers [
   {rendererClass=&NiagaraSpriteRendererProperties}
   {rendererClass=&NiagaraSpriteRendererProperties, rendererIndex=1}
  ]
 }
 FN_Debris_Sprite_SubUV {
  simTarget=GPUComputeSim
  emitterUpdateScript [
   &EmitterScalability(
    Module.bApplySpawnCountScale=false
    Module.bApplySpawnCountScaleByDistanceFraction=false
    Module.bUseMaxDistance=false
    Module.bUseMinDistance=false
   ) {enabled=false}
   &EmitterLifeCycle() {enabled=false}
   &EmitterState(
    Enable Distance Culling=true
    Enable Visibility Culling=true
    Life Cycle Mode=::NewEnumerator1
    Loop Behavior=::NewEnumerator1
    Loop Duration=1
    MaxDistance=3500
    Module.bUseMaxDistance=true
    Module.bUseMinDistance=false
    Scalability Mode=::NewEnumerator1
    UseLoopDelay=false
   )
   &SpawnBurst_Instantaneous(
    Spawn Count=&RandomRangeInt(Minimum=15, Maximum=25)
    Use Spawn Probability=false
    UseLoopCountLimit=false
   )
  ]
  particleSpawnScript [
   /Niagara/Modules/Spawn/Initialization/InitializeParticle(
    Module.Write Color=true
    Module.Write Lifetime=true
    Module.Write Mass=true
    Module.Write Material Random=true
    Module.Write Position=false
    Module.Write Scale=false
    Module.Write SpriteRotation=true
    Module.Write SpriteSize=true
    Position=&SimulationPosition()
    Sprite Rotation=&UniformRangedFloat(Minimum=0, Maximum=360)
    Sprite Size=&Vector2DFromFloat(Value=&UniformRangedFloat(Minimum=20, Maximum=50))
   ) {enabled=false}
   /Niagara/Modules/Spawn/Initialization/V2/InitializeParticle(
    Color Maximum=<rgba 0.6927 0.6914 0.6119 1>
    Color Minimum=<rgba 0.8021 0.789 0.7281 1>
    Color Mode=::NewEnumerator2
    Lifetime Max=1
    Lifetime Min=0.4
    Lifetime Mode=::NewEnumerator1
    Sprite Size Mode=::NewEnumerator1
    Uniform Sprite Size Max=4
    Uniform Sprite Size Min=2
    UsePositionOffset=false
    Write Material Random=true
   ) {module=InitializeParticle001}
   &AddVelocityInCone(
    Cone Angle=135
    Module.Use Velocity Falloff On Cone Axis=false
    Velocity Strength=&UniformRangedFloat(Minimum=300, Maximum=700)
   ) {enabled=false}
   &AddVelocity(
    Module.Use Falloff Distance=false
    Origin Offset=<0 0 -1>
    Override Seed=false
    Velocity Mode=::NewEnumerator1
    Velocity Speed=&FloatFromCurve(
     FloatCurve=DataInterface() {propertyValues {Curve {keys [
         {interpMode=RCIM_Cubic, tangentMode=RCTM_User, time=0, value=1}
         {arriveTangent=-1.3709, interpMode=RCIM_Cubic, leaveTangent=-1.3709, tangentMode=RCTM_User, time=1, value=0}
        ]}}}
     Scale Curve=&RandomRangeFloat(Minimum=50, Maximum=100)
    )
   ) {enabled=false}
   /SetVariables(Particles.RandomGravity:NiagaraFloat=&UniformRangedFloat(Minimum=-1200, Maximum=-2000)) {isSetParameters=true}
   &SubUVAnimation(
    Evaluation Type=::NewEnumerator0
    "Sprite Renderer "=DataInterface() {propertyValues {SpriteRenderer=&self:FN_Debris_Sprite_SubUV.NiagaraSpriteRendererProperties_0}}
    SubUV Animation Mode:ENiagara_SubUVLookupModeV2=::NewEnumerator2
    UseEndFrame=false
    UseStartFrame=false
   )
   &ApplyInitialForces()
   &AddRotationalVelocity(Rotation Rate=&Multiply_VectorByFloat(
     Vector=&RandomVector(Vector Scale=2)
     Float=&RandomRangeFloat(Minimum=0.5, Maximum=2)
    ))
   &AddVelocity(
    Cone Axis=<0.2 0 0.9>
    Module.UseConeFalloff=false
    Module.UseVelDistribution=false
    Override Seed=false
    Velocity Mode=::NewEnumerator2
    Velocity Speed=&RandomRangeFloat(Minimum=100, Maximum=390)
   ) {module=AddVelocity001}
  ]
  particleUpdateScript [
   &ParticleState()
   &ScaleSpriteSize(Uniform Curve Sprite Scale=DataInterface() {propertyValues {Curve {keys [
        {arriveTangent=9.0725, interpMode=RCIM_Cubic, leaveTangent=9.0725, tangentMode=RCTM_User, time=0, value=0}
        {arriveTangent=-2.5249, interpMode=RCIM_Cubic, leaveTangent=-2.5249, tangentMode=RCTM_User, time=0.1408, value=0.9847}
        {arriveTangent=-0.0405, interpMode=RCIM_Cubic, leaveTangent=-0.0405, tangentMode=RCTM_User, time=1, value=0}
       ]}}})
   &AccelerationForce(Acceleration=&MakeVector(X=0, Y=0, Z=$Particles.RandomGravity))
   &Drag(
    Drag=&FloatFromCurve(
     FloatCurve=DataInterface() {propertyValues {Curve {keys [
         {interpMode=RCIM_Cubic, tangentMode=RCTM_User, time=0, value=0}
         {interpMode=RCIM_Cubic, tangentMode=RCTM_User, time=0.5, value=1}
         {tangentMode=RCTM_User, time=1, value=0}
        ]}}}
     Scale Curve=15
    )
    Use Linear Drag=true
    Use Rotational Drag=false
   )
   &GravityForce(Gravity=&RandomRangeVector(Minimum=<0 0 -980>, Maximum=<0 0 -1200>))
   &SolveForcesAndVelocity(Clamp Velocity=false, Limit Acceleration=false)
   &ScaleSpriteSizeBySpeed(Max Scale Factor=<0.75 2>)
   &Color() {enabled=false}
   &ScaleColor(ScaleA=true, ScaleRGB=false)
  ]
  rendererClasses [&NiagaraSpriteRendererProperties]
  renderers [{rendererClass=&NiagaraSpriteRendererProperties}]
 }
 Flash_Flare {
  emitterUpdateScript [
   &EmitterState(
    Enable Distance Culling=true
    Enable Visibility Culling=true
    Max Distance Response=::KillImmediately
    Module.bUseMaxDistance=true
    Module.bUseMinDistance=false
    Scalability Mode=::NewEnumerator1
   )
   &SpawnBurst_Instantaneous(Spawn Count=1, Use Spawn Probability=false)
  ]
  particleSpawnScript [
   &InitializeParticle(
    Color=<rgba 1 0.8529 0.4844 1>
    Lifetime=0.2
    Sprite Size Mode=::NewEnumerator3
    Uniform Sprite Size=50
    UsePositionOffset=false
    Write Material Random=true
   ) {module=InitializeParticle001}
   &CameraOffset(Camera Offset Amount=3) {module=CameraOffset001}
  ]
  particleUpdateScript [
   &ParticleState()
   &ScaleSpriteSize(Uniform Curve Sprite Scale=DataInterface() {propertyValues {Curve {keys [
        {interpMode=RCIM_Cubic, tangentMode=RCTM_User, time=0, value=0}
        {interpMode=RCIM_Cubic, tangentMode=RCTM_User, time=0.5, value=1}
        {tangentMode=RCTM_User, time=1, value=0}
       ]}}})
   &SolveForcesAndVelocity(Clamp Velocity=false, Limit Acceleration=false)
   &ScaleColor(
    Scale Alpha=&FloatFromCurve(
     FloatCurve=DataInterface() {propertyValues {Curve {keys [
         {arriveTangent=9.0725, interpMode=RCIM_Cubic, leaveTangent=9.0725, tangentMode=RCTM_User, time=0, value=0}
         {arriveTangent=-6.7489, interpMode=RCIM_Cubic, leaveTangent=-6.7489, tangentMode=RCTM_User, time=0.1408, value=3}
         {arriveTangent=-0.0405, interpMode=RCIM_Cubic, leaveTangent=-0.0405, tangentMode=RCTM_User, time=1, value=0}
        ]}}}
     Scale Curve=0.1
    )
    Scale RGB=&VectorFromFloat(Value=&FloatFromCurve(
      FloatCurve=DataInterface() {propertyValues {Curve {keys [
          {arriveTangent=9.0725, interpMode=RCIM_Cubic, leaveTangent=9.0725, tangentMode=RCTM_User, time=0, value=1}
          {arriveTangent=-26.163, interpMode=RCIM_Cubic, leaveTangent=-26.163, tangentMode=RCTM_User, time=0.1, value=15}
          {arriveTangent=-0.0405, interpMode=RCIM_Cubic, leaveTangent=-0.0405, tangentMode=RCTM_User, time=1, value=1}
         ]}}}
      Scale Curve=2
     ))
    ScaleA=true
    ScaleRGB=true
   )
  ]
  rendererClasses [&NiagaraSpriteRendererProperties]
  renderers [{rendererClass=&NiagaraSpriteRendererProperties}]
 }
 RocksBurst {
  simTarget=GPUComputeSim
  emitterUpdateScript [
   &EmitterState(
    Enable Distance Culling=true
    Enable Visibility Culling=true
    Life Cycle Mode=::NewEnumerator1
    Loop Behavior=::NewEnumerator1
    Loop Duration=1
    Max Distance Response=::KillImmediately
    MaxDistance=4500
    Module.bApplySpawnCountScaleByDistance=true
    Module.bUseMaxDistance=true
    Module.bUseMinDistance=false
    Scalability Mode=::NewEnumerator1
    Scale Spawn Count=true
    Spawn Count Scale By Distance=DataInterface() {propertyValues {Curve {keys [<at 0.3676 0.9657>,<at 0.9962 0.4982>]}}}
    UseLoopDelay=false
   )
   &SpawnBurst_Instantaneous(
    Spawn Count=&RandomRangeInt(Minimum=7, Maximum=11)
    Use Spawn Probability=false
    UseLoopCountLimit=false
   )
  ]
  particleSpawnScript [
   &InitializeParticle(
    Color Maximum=<rgba 0.6927 0.6914 0.6119 1>
    Color Minimum=<rgba 0.8021 0.789 0.7281 1>
    Color Mode=::NewEnumerator2
    Lifetime Max=1.3
    Lifetime Min=0.75
    Lifetime Mode=::NewEnumerator1
    Mesh Scale Max=<0.07 0.07 0.2>
    Mesh Scale Min=<0.02 0.02 0.02>
    Mesh Scale Mode=::NewEnumerator2
    UsePositionOffset=false
    Write Material Random=true
   )
   &ShapeLocation(MaskSpawnGroup=false, Sphere Distribution=::NewEnumerator3, Sphere Radius=1, UseOwnerScale=true)
   &InitialMeshOrientation(Use Rotation Vector=false)
   &self:DivideParticles(HowManyRenderers=3) {module=DivideParticles}
   &AddRotationalVelocity(Rotation Rate=&Multiply_VectorByFloat(
     Vector=&RandomVector(Vector Scale=2)
     Float=&RandomRangeFloat(Minimum=0.5, Maximum=2)
    ))
   &ApplyInitialForces()
   /SetVariables(Particles.Scale=&Multiply_VectorByFloat(Vector=$Particles.Scale, Float=0.7)) {isSetParameters=true}
   &AddVelocity(
    Cone Axis=<0.2 0 1>
    Module.UseConeFalloff=false
    Module.UseVelDistribution=false
    Override Seed=false
    Velocity Mode=::NewEnumerator2
    Velocity Speed=&RandomRangeFloat(Minimum=100, Maximum=390)
   ) {module=AddVelocity001}
  ]
  particleUpdateScript [
   &ParticleState()
   &GravityForce()
   &Drag(
    Drag=&FloatFromCurve(
     FloatCurve=DataInterface() {propertyValues {Curve {keys [
         {interpMode=RCIM_Cubic, tangentMode=RCTM_User, time=0, value=0}
         {interpMode=RCIM_Cubic, tangentMode=RCTM_User, time=0.5, value=1}
         {tangentMode=RCTM_User, time=1, value=0}
        ]}}}
     Scale Curve=1.5
    )
    Use Linear Drag=true
    Use Rotational Drag=false
   )
   &SolveForcesAndVelocity(Clamp Velocity=false, Limit Acceleration=false)
   &ScaleMeshSize(Scale Factor=&VectorFromFloat(Value=&FloatFromCurve(FloatCurve=DataInterface() {propertyValues {Curve {keys [
          {arriveTangent=9.0725, interpMode=RCIM_Cubic, leaveTangent=9.0725, tangentMode=RCTM_User, time=0, value=0}
          {arriveTangent=0.4555, interpMode=RCIM_Cubic, leaveTangent=0.4555, tangentMode=RCTM_User, time=0.1408, value=1.2}
          {arriveTangent=0.068, interpMode=RCIM_Cubic, leaveTangent=0.068, tangentMode=RCTM_User, time=1, value=0}
         ]}}})))
  ]
  rendererClasses [&NiagaraMeshRendererProperties]
  renderers [
   {rendererClass=&NiagaraMeshRendererProperties, rendererIndex=0}
   {rendererClass=&NiagaraMeshRendererProperties, rendererIndex=1}
   {rendererClass=&NiagaraMeshRendererProperties, rendererIndex=2}
  ]
 }
 Sparks {
  simTarget=GPUComputeSim
  emitterUpdateScript [
   &EmitterState(
    Enable Distance Culling=true
    Enable Visibility Culling=true
    Inactive Response=::NewEnumerator1
    Life Cycle Mode=::NewEnumerator1
    Loop Behavior=::NewEnumerator1
    Loop Duration=1
    MaxDistance=3000
    Module.bUseMaxDistance=true
    Module.bUseMinDistance=false
    Scalability Mode=::NewEnumerator1
    UseLoopDelay=false
   )
   &SpawnBurst_Instantaneous(
    Spawn Count=&RandomRangeInt(Minimum=3, Maximum=5)
    Use Spawn Probability=false
    UseLoopCountLimit=false
   )
  ]
  particleSpawnScript [
   &InitializeParticle(
    Color Maximum=<rgba 0.3333 0.2436 0.0563 1>
    Color Minimum=<rgba 1 0.6883 0.3554 1>
    Color Mode=::NewEnumerator2
    Lifetime Max=0.55
    Lifetime Min=0.166
    Lifetime Mode=::NewEnumerator1
    Sprite Size Max=<3 5>
    Sprite Size Min=<2 4>
    Sprite Size Mode=::NewEnumerator2
    UsePositionOffset=false
    Write Material Random=true
   )
   &ShapeLocation(MaskSpawnGroup=false, Override Seed=false, Sphere Radius=1, Sphere Surface Distribution=1, UseOwnerScale=true)
   &AddVelocity(
    Cone Angle=100
    Cone Axis=<1 0 0.5>
    Module.UseConeFalloff=false
    Module.UseVelDistribution=false
    Override Seed=false
    Velocity Mode=::NewEnumerator2
    Velocity Speed=&RandomRangeFloat(Minimum=222, Maximum=666)
   )
  ]
  particleUpdateScript [
   &ParticleState()
   &GravityForce(Gravity=<0 0 -666>)
   &ScaleSpriteSize(Scale Factor=&Vector2DFromCurve(
     Vector2Curve=DataInterface() {propertyValues {
       XCurve {keys [
         {arriveTangent=5.9728, interpMode=RCIM_Cubic, leaveTangent=5.9728, tangentMode=RCTM_User, time=0, value=0.1}
         {arriveTangent=-0.1, interpMode=RCIM_Cubic, leaveTangent=-0.1, tangentMode=RCTM_User, time=0.18, value=0.8}
         {interpMode=RCIM_Cubic, tangentMode=RCTM_User, time=1, value=0}
        ]}
       YCurve {keys [
         {arriveTangent=6.2715, interpMode=RCIM_Cubic, leaveTangent=6.2715, tangentMode=RCTM_User, time=0, value=0.1}
         {arriveTangent=-0.1001, interpMode=RCIM_Cubic, leaveTangent=-0.1001, tangentMode=RCTM_User, time=0.18, value=0.8}
         {interpMode=RCIM_Cubic, tangentMode=RCTM_User, time=0.9988, value=0}
        ]}
      }}
     ScaleCurve=<1 0.4>
    ))
   &Drag(Drag=&UniformRangedFloat(Minimum=2, Maximum=6), Use Linear Drag=true, Use Rotational Drag=false)
   &CurlNoiseForce(Noise Strength=100, Pan Noise=true, Pan Noise Field=<0 0.2 0.2>)
   &SolveForcesAndVelocity(Clamp Velocity=false, Limit Acceleration=false)
   &ScaleSpriteSizeBySpeed(
    Max Scale Factor=&Vector2DFromCurve(
     Vector2Curve=DataInterface() {propertyValues {
       XCurve {keys [<at 0.0046 10>,<at 0.1 1.25>,<at 1 1>]}
       YCurve {keys [
         {arriveTangent=-150.263, interpMode=RCIM_Cubic, leaveTangent=-150.263, tangentMode=RCTM_User, time=0, value=20}
         {arriveTangent=-24, interpMode=RCIM_Cubic, leaveTangent=2.272, tangentMode=RCTM_Break, time=0.2, value=1.25}
         <at 1 1 RCIM_Cubic>
        ]}
      }}
     ScaleCurve=<0.45 8>
    )
    Velocity Threshold=666
   )
   &ScaleColor(
    Scale Alpha=&FloatFromCurve(
     FloatCurve=DataInterface() {propertyValues {Curve {keys [
         {interpMode=RCIM_Cubic, tangentMode=RCTM_User, time=0, value=0}
         {interpMode=RCIM_Cubic, tangentMode=RCTM_User, time=0.5, value=1}
         {tangentMode=RCTM_User, time=1, value=0}
        ]}}}
     Scale Curve=&MakeCustomFloatFromBool(Bool=$User.bIsFrontEndPreview, True Float=0.5, False Float=1)
    )
    Scale RGB=&VectorFromFloat(Value=&FloatFromCurve(FloatCurve=DataInterface() {propertyValues {Curve {keys [<at 0 1>,<at 0.1 10>,<at 0.2 1.25>,<at 1 1>]}}}))
    ScaleA=true
    ScaleRGB=true
   )
  ]
  rendererClasses [&NiagaraSpriteRendererProperties]
  renderers [
   {rendererClass=&NiagaraSpriteRendererProperties}
   {rendererClass=&NiagaraSpriteRendererProperties, rendererIndex=1}
  ]
 }
 SpikeStreaksMesh {
  emitterUpdateScript [
   &EmitterState(
    Enable Distance Culling=true
    Enable Visibility Culling=true
    Max Distance Response=::KillImmediately
    Module.bUseMaxDistance=true
    Module.bUseMinDistance=false
    Scalability Mode=::NewEnumerator1
   )
   &SpawnBurst_Instantaneous(Spawn Count=9, Use Spawn Probability=false)
  ]
  particleSpawnScript [
   /Niagara/Modules/Spawn/Initialization/InitializeParticle(
    Color=<rgba 0.4167 0.3538 0.1674 1>
    Lifetime=0.25
    Mesh Scale=&Multiply_VectorByFloat(Vector=&UniformRangedVector(Minimum=<0.9 0.6 0.3>, Maximum=<1.5 0.6 0.3>), Float=0.475)
    Module.Write Color=true
    Module.Write Lifetime=true
    Module.Write Mass=true
    Module.Write Material Random=true
    Module.Write Position=true
    Module.Write Scale=true
    Module.Write SpriteRotation=false
    Module.Write SpriteSize=false
    Position=&Add_Vector(A=/Niagara/DynamicInputs/Helpers/SimulationPosition(), B=<0 0 0>)
   )
   /Game/Effects/Niagara/Modules/Orientation/FN_Fan(
    Cone Angle:NiagaraFloat=&UniformRangedFloat(Minimum=150, Maximum=180)
    Spin:NiagaraFloat=&UniformRangedFloat(Minimum=0, Maximum=1)
   ) {enabled=false}
   &CameraOffset(Camera Offset Amount=3) {module=CameraOffset001}
   &AddVelocity(
    Cone Angle=100
    Cone Axis=<1 0 0.5>
    Module.UseConeFalloff=false
    Module.UseVelDistribution=false
    Override Seed=false
    Velocity Mode=::NewEnumerator2
    Velocity Speed=1
   )
  ]
  particleUpdateScript [
   &SolveForcesAndVelocity(Clamp Velocity=false, Limit Acceleration=false)
   &ParticleState()
   &CameraOffset()
   &ScaleMeshSize(Scale Factor=&VectorFromCurve(VectorCurve=DataInterface() {propertyValues {
       bOptimizeLUT=false
       XCurve {keys [{arriveTangent=1.8635, interpMode=RCIM_Cubic, leaveTangent=1.8635, tangentMode=RCTM_User, time=-0.0048, value=0.4}, <at 0.9737 1>]}
       YCurve {keys [
         {arriveTangent=0.1613, interpMode=RCIM_Cubic, leaveTangent=0.1613, tangentMode=RCTM_User, time=0, value=1}
         {arriveTangent=-1.3776, interpMode=RCIM_Cubic, leaveTangent=-1.3776, tangentMode=RCTM_User, time=0.9944, value=0}
        ]}
       ZCurve {keys [<at 0 1 RCIM_Cubic>, {arriveTangent=-1.635, interpMode=RCIM_Cubic, leaveTangent=-1.635, tangentMode=RCTM_User, time=0.9961, value=0}]}
      }}))
   /Game/Effects/Niagara/Modules/Materials/MaterialPanner(Offset:NiagaraFloat=&Add_Float(A=$Emitter.Age, B=$Particles.Initial.MaterialRandom), Speed=-5)
   &ScaleColor(
    Scale Alpha=&FloatFromCurve(FloatCurve=DataInterface() {propertyValues {Curve {keys [
         {arriveTangent=-4.0245, interpMode=RCIM_Cubic, leaveTangent=-4.0245, tangentMode=RCTM_User, time=0, value=1.5}
         {arriveTangent=-1.5, interpMode=RCIM_Cubic, leaveTangent=-1.5, time=0.2, value=0.8}
         <at 1 0>
        ]}}})
    Scale RGB=&VectorFromFloat(Value=&FloatFromCurve(
      FloatCurve=DataInterface() {propertyValues {Curve {keys [
          {arriveTangent=-4.0245, interpMode=RCIM_Cubic, leaveTangent=-4.0245, tangentMode=RCTM_User, time=0, value=1.5}
          {arriveTangent=-1.5, interpMode=RCIM_Cubic, leaveTangent=-1.5, time=0.2, value=0.8}
          <at 1 0>
         ]}}}
      Scale Curve=2
     ))
    ScaleA=true
    ScaleRGB=true
   )
  ]
  rendererClasses [&NiagaraMeshRendererProperties]
  renderers [{rendererClass=&NiagaraMeshRendererProperties, rendererIndex=0}]
 }
 SpikeStreaksParticle {
  emitterUpdateScript [
   &EmitterState(
    Enable Distance Culling=true
    Enable Visibility Culling=true
    Max Distance Response=::KillImmediately
    Module.bUseMaxDistance=true
    Module.bUseMinDistance=false
    Scalability Mode=::NewEnumerator1
   )
   &SpawnBurst_Instantaneous(Spawn Count=11, Use Spawn Probability=false)
  ]
  particleSpawnScript [
   &InitializeParticle(
    Color=<rgba 0.4167 0.3538 0.1674 1>
    Lifetime=&UniformRangedFloat(Minimum=0.125, Maximum=0.15)
    Module.Write Color=true
    Module.Write Lifetime=true
    Module.Write Mass=true
    Module.Write Material Random=true
    Module.Write Position=true
    Module.Write Scale=true
    Module.Write SpriteRotation=false
    Module.Write SpriteSize=true
    Position=&Add_Vector(A=&SimulationPosition(), B=<0 0 0>)
    Sprite Size=&Multiply_Vector2D_ByFloat(Float=0.6, Vector2D=&UniformRangedVector2D(Minimum=<40 62>, Maximum=<20 50>))
   ) {enabled=false}
   /Niagara/Modules/Spawn/Initialization/V2/InitializeParticle(
    Color=<rgba 0.4167 0.3538 0.1674 1>
    Lifetime=&UniformRangedFloat(Minimum=0.15, Maximum=0.175)
    Sprite Size Max=&Multiply_Vector2D_ByFloat(Float=0.65, Vector2D=<20 50>)
    Sprite Size Min=&Multiply_Vector2D_ByFloat(Float=0.6, Vector2D=<40 62>)
    Sprite Size Mode=::NewEnumerator2
    UsePositionOffset=false
    Write Material Random=true
   ) {module=InitializeParticle001}
   &ShapeLocation(MaskSpawnGroup=false, Sphere Distribution=::NewEnumerator1, Sphere Radius=0, UseOwnerScale=true)
   &AddVelocityFromPoint(Module.Use Falloff Distance=false) {enabled=false}
   &CameraOffset(Camera Offset Amount=3) {module=CameraOffset001}
   &AddVelocity(
    Cone Angle=150
    Cone Axis=<1 0 0.5>
    Module.UseConeFalloff=false
    Module.UseVelDistribution=false
    Override Seed=false
    Velocity Mode=::NewEnumerator2
    Velocity Speed=75
   )
  ]
  particleUpdateScript [
   &ParticleState()
   &CameraOffset()
   &SolveForcesAndVelocity(Clamp Velocity=false, Limit Acceleration=false)
   &ScaleSpriteSize(Scale Factor=&Vector2DFromFloat(Value=&FloatFromCurve(FloatCurve=DataInterface() {propertyValues {Curve {keys [<at 0 0.6>,<at 1 1.7>]}}}, Scale Curve=0.75)))
   &ScaleColor(
    Scale Alpha=&FloatFromCurve(
     FloatCurve=DataInterface() {propertyValues {Curve {keys [
         {arriveTangent=-4.0245, interpMode=RCIM_Cubic, leaveTangent=-4.0245, tangentMode=RCTM_User, time=0, value=1.5}
         {arriveTangent=-1.5, interpMode=RCIM_Cubic, leaveTangent=-1.5, time=0.2, value=0.8}
         <at 1 0>
        ]}}}
     Scale Curve=0.6
    )
    Scale RGB=&VectorFromFloat(Value=0.6)
    ScaleA=true
    ScaleRGB=true
   )
  ]
  rendererClasses [&NiagaraSpriteRendererProperties]
  renderers [{rendererClass=&NiagaraSpriteRendererProperties}]
 }
}
summary {
 systemName=NS_Emote_InflatoDodo_Snap
 emitters [
  {emitterName=SpikeStreaksMesh, rendererClasses [&NiagaraMeshRendererProperties]}
  {emitterName=Flash_Flare, rendererClasses [&NiagaraSpriteRendererProperties]}
  {emitterName=SpikeStreaksParticle, rendererClasses [&NiagaraSpriteRendererProperties]}
  {emitterName=Sparks, simTarget=GPUComputeSim, rendererClasses [&NiagaraSpriteRendererProperties]}
  {emitterName=RocksBurst, simTarget=GPUComputeSim, rendererClasses [&NiagaraMeshRendererProperties]}
  {emitterName=FN_Debris_Sprite_SubUV, simTarget=GPUComputeSim, rendererClasses [&NiagaraSpriteRendererProperties]}
  {emitterName=Dust, rendererClasses [&NiagaraSpriteRendererProperties]}
 ]
 userVariables [
  User.bIsFrontEndPreview {type=NiagaraBool}
  User.Impact_Size {type=NiagaraFloat, defaultValue=0.5, description=""}
 ]
}
